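; Vehicle cosmetic drift trail Blueprint for the StarBeam variant, which applies paint color to its Niagara trail effect.
; The SetupPaintSettings override calls the parent to resolve the primary color, balanced and metallic values from the paint name and additional info.
; It writes the primary color to the Shape01 and Shape02 Niagara variables, fixes ColorWhite at full, then returns the parent's results unchanged.
; Event graph and construction script add no other logic.
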(event EventBeginPlay
  (|CallParentFunction) ; This node is disabled and will not be called.\nDrag off pins to build functionality.
)

(event Collision|EventActorBeginOverlap
  (params (OtherActor :type Actor%20Object%20Reference)))

(event EventTick
  (params (DeltaSeconds :type Float%20%28single-precision%29)))

(fn SetupPaintSettings
  (params (Color :type Name :default "None") (AdditionalInfo :type Vehicle%20Cosmetics%20Additional%20Parameter%20Infos%20Structure))
  (bind _ns_drifttrail (Variables|Default|GetNS_DriftTrail))
  (bind (_primarycolor _balanced _metallic) (|CallParentFunction Color AdditionalInfo))
  (Niagara|SetNiagaraVariable(LinearColor) _ns_drifttrail "Shape01" _primarycolor)
  (Niagara|SetNiagaraVariable(LinearColor) _ns_drifttrail "Shape02" _primarycolor)
  (Niagara|SetNiagaraVariable(Float) _ns_drifttrail "ColorWhite" 1.0)
  (return _primarycolor _balanced _metallic))

(fn ConstructionScript
  (|CallParentFunction))
