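# Niagara system for the StarBeam vehicle drift trail cosmetic, attaching stars, ribbons and spinning wheel meshes to the car's skeletal mesh.
# AllStars merges four spawn groups into one sprite emitter, using group masks for fast backward streaks and slower gravity-affected stars with separate SubUV frames.
# Ribbon draws two fading trails, and Wheel scales meshes to the wheels.
# Painted-variant overrides recolor all emitters, while stars and ribbons are disabled per tire via FN_DisableTireFX.
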NiagaraSystem /VehicleCosmetics/DriftTrail/DriftTrail_StarBeam/FX/DriftTrail_StarBeam_MP
dependencies {
 usedDataInterfaces [
  &NiagaraDataInterfaceCurve
  &NiagaraDataInterfaceSpriteRendererInfo
  &NiagaraDataInterfaceArrayInt32
  &NiagaraDataInterfaceVectorCurve
  &NiagaraDataInterfaceSkeletalMesh
 ]
 usedDynamicInputs []
 usedModules [
  &self:SystemSpawnScript.SystemScriptSource.SystemScriptGraph.NiagaraNodeAssignment_0.SetVariables_#1
  &SystemState
  &FN_UnpackVehicleFlags
  /Game/Effects/Niagara/Modules/Vehicle/FN_UpdateVehicleAttribs
  /Game/Effects/Niagara/Modules/Vehicle/FN_VehicleFlags_Cheatcodes
 ]
}
emitters {
 AllStars {
  emitterUpdateScript [
   &EmitterState(Inactive Response=::NewEnumerator1, Life Cycle Mode=::NewEnumerator1, Loop Duration=1, UseLoopDelay=false)
   &SpawnRate(
    SpawnRate=/VehicleCosmetics/FX/Niagara/Modules/VC_Mask_Floats(
     FrontEnd Value:NiagaraFloat=&RandomRangeFloat(Minimum=300, Maximum=500)
     blsFrontEndPreview:NiagaraBool=$User.bIsFrontendPreview
     SpeedMin Value:NiagaraFloat=&RandomRangeFloat(Minimum=300, Maximum=500)
     SpeedMax Value:NiagaraFloat=&RandomRangeFloat(Minimum=300, Maximum=500)
     Speed Alpha:NiagaraFloat=$System.Speed_Alpha
    )
    Use Spawn Probability=false
   )
   &SpawnRate(
    SpawnGroup=1
    SpawnRate=&VC_Mask_Floats(
     FrontEnd Value:NiagaraFloat=&RandomRangeFloat(Minimum=300, Maximum=500)
     blsFrontEndPreview:NiagaraBool=$User.bIsFrontendPreview
     SpeedMin Value:NiagaraFloat=&RandomRangeFloat(Minimum=300, Maximum=500)
     SpeedMax Value:NiagaraFloat=&RandomRangeFloat(Minimum=300, Maximum=500)
     Speed Alpha:NiagaraFloat=$System.Speed_Alpha
    )
    Use Spawn Probability=false
   ) {module=SpawnRate001}
   &SpawnRate(
    SpawnGroup=2
    SpawnRate=&VC_Mask_Floats(
     FrontEnd Value:NiagaraFloat=&RandomRangeFloat(Minimum=50, Maximum=40)
     blsFrontEndPreview:NiagaraBool=$User.bIsFrontendPreview
     SpeedMin Value:NiagaraFloat=&RandomRangeFloat(Minimum=50, Maximum=40)
     SpeedMax Value:NiagaraFloat=&RandomRangeFloat(Minimum=50, Maximum=40)
     Speed Alpha:NiagaraFloat=$System.Speed_Alpha
    )
    Use Spawn Probability=false
   ) {module=SpawnRate002}
   &SpawnRate(
    SpawnGroup=3
    SpawnRate=&VC_Mask_Floats(
     FrontEnd Value:NiagaraFloat=&RandomRangeFloat(Minimum=50, Maximum=40)
     blsFrontEndPreview:NiagaraBool=$User.bIsFrontendPreview
     SpeedMin Value:NiagaraFloat=&RandomRangeFloat(Minimum=50, Maximum=40)
     SpeedMax Value:NiagaraFloat=&RandomRangeFloat(Minimum=50, Maximum=40)
     Speed Alpha:NiagaraFloat=$System.Speed_Alpha
    )
    Use Spawn Probability=false
   ) {module=SpawnRate003}
  ]
  particleSpawnScript [
   &VC_PaintedVariantsOverride(
    BurntSienna=<rgba 0.0642 0.0198 0.0079 1>
    DefaultColor=<rgba 0 0.7765 1 1>
    Orange=<rgba 0.6724 0.2346 0 1>
    R1 Black:LinearColor=<rgba 0.05 0.05 0.05 1>
    R1 BurntSienna:LinearColor=<rgba 0.14 0.0432 0.0173 1>
    R1 Cobalt:LinearColor=<rgba 0 0.0185 1 1>
    R1 DefaultColor=<rgba 1 0.9167 0.15 1>
    R1 Orange:LinearColor=<rgba 0.672 0.3584 0 1>
    R1 Pink:LinearColor=<rgba 0.8 0.24 0.5964 1>
    R1 Purple:LinearColor=<rgba 0.8 0.24 0.5964 1>
    Use R1 Painted Overrides=true
   )
   /SetVariables(Particles.TagStars:NiagaraInt32=&SpawnGroup()) {isSetParameters=true}
   &InitializeParticle(
    Color Mode=::NewEnumerator0
    Lifetime Max=/VehicleCosmetics/FX/Niagara/Modules/MergeEmitters/VC_FloatBySpawnGroup(
     SpawnGroup or ExecIndex:NiagaraInt32=0
     CustomInt:NiagaraInt32=0
     MaximumSpawnGroups:NiagaraInt32=4
     Float SG 0:NiagaraFloat=0.16
     Float SG 1:NiagaraFloat=0.15
     Float SG 2:NiagaraFloat=0.65
     Float SG 3:NiagaraFloat=0.65
     Float SG 4:NiagaraFloat=0
     Float SG 5:NiagaraFloat=0
    )
    Lifetime Min:NiagaraFloat=&VC_FloatBySpawnGroup(
     SpawnGroup or ExecIndex:NiagaraInt32=0
     CustomInt:NiagaraInt32=0
     MaximumSpawnGroups:NiagaraInt32=4
     Float SG 0:NiagaraFloat=0.05
     Float SG 1:NiagaraFloat=0.03
     Float SG 2:NiagaraFloat=0.45
     Float SG 3:NiagaraFloat=0.45
     Float SG 4:NiagaraFloat=0
     Float SG 5:NiagaraFloat=0
    )
    Lifetime Mode=::NewEnumerator1
    Sprite Rotation Mode=::NewEnumerator2
    Sprite Size Mode=::NewEnumerator1
    Uniform Sprite Size Max=15
    Uniform Sprite Size Min=10
    UsePositionOffset=false
    Write Material Random=true
   )
   &PartitionParticles()
   &SkeletalMeshLocation(
    Bone / Socket Index=&Add_Integer(A=$Particles.PartitionParticles.Partition, B=4)
    Bone / Socket Sampling Mode=::NewEnumerator5
    Interpolate Mesh Sampling=false
    MaskSpawnGroup=false
    Mesh Sampling Type=::NewEnumerator6
    Sampled Position Offset=&AddVectorToPosition(Position=<-35 0 -10>, Vector=$User.offset)
    Skeletal Mesh=$System.SkeletalMesh
   )
   &AddVelocity(Velocity=&VC_Mask_Vectors(
     FrontEnd Value:Vector3f=&MaskVectorBySpawnGroup(Masked Vector=<-5000 0 0>, Passthrough Vector=<1 1 1>)
     blsFrontEndPreview:NiagaraBool=$User.bIsFrontendPreview
     SpeedMin Value:Vector3f=&MaskVectorBySpawnGroup(Masked Vector=<-5000 0 0>, Passthrough Vector=<1 1 1>)
     SpeedMax Value:Vector3f=&MaskVectorBySpawnGroup(Masked Vector=<-5000 0 0>, Passthrough Vector=<1 1 1>)
     Speed Alpha:NiagaraFloat=$System.Speed_Alpha
    ))
   &AddVelocity(Velocity=&VC_Mask_Vectors(
     FrontEnd Value:Vector3f=&MaskVectorBySpawnGroup(Masked Vector=<-5000 0 0>, Passthrough Vector=<1 1 1>, Spawn Group Mask=1)
     blsFrontEndPreview:NiagaraBool=$User.bIsFrontendPreview
     SpeedMin Value:Vector3f=&MaskVectorBySpawnGroup(Masked Vector=<-5000 0 0>, Passthrough Vector=<1 1 1>, Spawn Group Mask=1)
     SpeedMax Value:Vector3f=&MaskVectorBySpawnGroup(Masked Vector=<-5000 0 0>, Passthrough Vector=<1 1 1>, Spawn Group Mask=1)
     Speed Alpha:NiagaraFloat=$System.Speed_Alpha
    )) {module=AddVelocity001}
   &AddVelocity(Velocity=&VC_Mask_Vectors(
     FrontEnd Value:Vector3f=&MaskVectorBySpawnGroup(
      Masked Vector=&RandomRangeVector(Minimum=<-600 -30 300>, Maximum=<-350 30 0>)
      Passthrough Vector=<1 1 1>
      Spawn Group Mask=2
     )
     blsFrontEndPreview:NiagaraBool=$User.bIsFrontendPreview
     SpeedMin Value:Vector3f=&MaskVectorBySpawnGroup(
      Masked Vector=&RandomRangeVector(Minimum=<-600 -30 300>, Maximum=<-350 30 0>)
      Passthrough Vector=<1 1 1>
      Spawn Group Mask=2
     )
     SpeedMax Value:Vector3f=&MaskVectorBySpawnGroup(
      Masked Vector=&RandomRangeVector(Minimum=<-600 -30 300>, Maximum=<-350 30 0>)
      Passthrough Vector=<1 1 1>
      Spawn Group Mask=2
     )
     Speed Alpha:NiagaraFloat=$System.Speed_Alpha
    )) {module=AddVelocity002}
   &AddVelocity(Velocity=&VC_Mask_Vectors(
     FrontEnd Value:Vector3f=&MaskVectorBySpawnGroup(
      Masked Vector=&RandomRangeVector(Minimum=<-600 -30 300>, Maximum=<-350 30 0>)
      Passthrough Vector=<1 1 1>
      Spawn Group Mask=3
     )
     blsFrontEndPreview:NiagaraBool=$User.bIsFrontendPreview
     SpeedMin Value:Vector3f=&MaskVectorBySpawnGroup(
      Masked Vector=&RandomRangeVector(Minimum=<-600 -30 300>, Maximum=<-350 30 0>)
      Passthrough Vector=<1 1 1>
      Spawn Group Mask=3
     )
     SpeedMax Value:Vector3f=&MaskVectorBySpawnGroup(
      Masked Vector=&RandomRangeVector(Minimum=<-600 -30 300>, Maximum=<-350 30 0>)
      Passthrough Vector=<1 1 1>
      Spawn Group Mask=3
     )
     Speed Alpha:NiagaraFloat=$System.Speed_Alpha
    )) {module=AddVelocity003}
   &ShapeLocation(
    MaskSpawnGroup=false
    Override Seed=false
    Sphere Radius=&RandomRangeFloat(Minimum=2, Maximum=5)
    UseOwnerScale=true
   )
   &SubUVAnimation(
    End Frame Range Override:NiagaraInt32=&SelectIntFromArray(Direct Array Index=&SpawnGroup(), Int Selection Array=DataInterface() {propertyValues {IntData [3,2,3,4]}})
    "Sprite Renderer "=DataInterface() {propertyValues {SpriteRenderer=&self:AllStars_3.NiagaraSpriteRendererProperties_4}}
    Start Frame Range Override=&SelectIntFromArray(Direct Array Index=$Emitter.SpawnGroup, Int Selection Array=DataInterface() {propertyValues {IntData [0,1,0,2]}})
    SubUV Animation Mode:ENiagara_SubUVLookupModeV2=::NewEnumerator2
    UseEndFrame=true
    UseStartFrame=true
   )
   &SubUVAnimation(
    End Frame Range Override:NiagaraInt32=&SelectIntFromArray(Direct Array Index=&SpawnGroup(), Int Selection Array=DataInterface() {propertyValues {IntData [3,2,3,4]}})
    "Sprite Renderer "=DataInterface() {propertyValues {SpriteRenderer=&self:AllStars_3.NiagaraSpriteRendererProperties_5}}
    Start Frame Range Override=&SelectIntFromArray(Direct Array Index=$Emitter.SpawnGroup, Int Selection Array=DataInterface() {propertyValues {IntData [0,1,0,2]}})
    SubUV Animation Mode:ENiagara_SubUVLookupModeV2=::NewEnumerator2
    UseEndFrame=true
    UseStartFrame=true
   ) {module=SubUVAnimation001}
   &SubUVAnimation(
    End Frame Range Override:NiagaraInt32=&SelectIntFromArray(
     Array Sampling Mode=::NewEnumerator1
     Direct Array Index=&SpawnGroup()
     Int Selection Array=DataInterface() {propertyValues {IntData [3,2,3,4]}}
     Random Seed=0
    )
    "Sprite Renderer "=DataInterface() {propertyValues {SpriteRenderer=&self:AllStars_3.NiagaraSpriteRendererProperties_1}}
    Start Frame Range Override=&SelectIntFromArray(
     Array Sampling Mode=::NewEnumerator1
     Direct Array Index=$Emitter.SpawnGroup
     Int Selection Array=DataInterface() {propertyValues {IntData [0,1,0,2]}}
     Random Seed=0
    )
    SubUV Animation Mode:ENiagara_SubUVLookupModeV2=::NewEnumerator2
    UseEndFrame=true
    UseStartFrame=true
   ) {module=SubUVAnimation002}
   &ShapeLocation(MaskSpawnGroup=false, Override Seed=false, Sphere Radius=20, UseOwnerScale=true) {module=ShapeLocation001}
  ]
  particleUpdateScript [
   &FN_DisableTireFX(
    Mask Index Attribute:NiagaraInt32=$Particles.PartitionParticles.Partition
    Mask Index FL=2
    Mask Index FR=3
    Mask Index RL=0
    Mask Index RR=1
   )
   &ParticleState()
   &ScaleSpriteSize(Uniform Curve Sprite Scale=DataInterface() {propertyValues {Curve {keys [{interpMode=RCIM_Cubic, tangentMode=RCTM_User, time=0, value=1}, {interpMode=RCIM_Cubic, tangentMode=RCTM_User, time=1, value=0}]}}})
   &GravityForce(Gravity=&MaskVectorBySpawnGroup(
     Masked Vector=&VectorFromCurve(
      VectorCurve=DataInterface() {propertyValues {
        XCurve {keys [<at 0 0>,<at 1 0>]}
        YCurve {keys [<at 0 0>,<at 1 0>]}
        ZCurve {keys [{interpMode=RCIM_Cubic, leaveTangent=1.1694, tangentMode=RCTM_Break, time=0, value=0.3}, <at 1 1>]}
       }}
      Scale Curve=<0 0 -980>
     )
     Passthrough Vector=<1 1 1>
     Spawn Group Mask=2
    ))
   &GravityForce(Gravity=&MaskVectorBySpawnGroup(
     Masked Vector=&VectorFromCurve(
      VectorCurve=DataInterface() {propertyValues {
        XCurve {keys [<at 0 0>,<at 1 0>]}
        YCurve {keys [<at 0 0>,<at 1 0>]}
        ZCurve {keys [{interpMode=RCIM_Cubic, leaveTangent=1.1694, tangentMode=RCTM_Break, time=0, value=0.3}, <at 1 1>]}
       }}
      Scale Curve=<0 0 -980>
     )
     Passthrough Vector=<1 1 1>
     Spawn Group Mask=3
    )) {module=GravityForce001}
   &InheritSourceMovement(
    Applied Position Scale=&Multiply_VectorByFloat(
     Vector=<1 1 1>
     Float=&FloatFromCurve(FloatCurve=DataInterface() {propertyValues {Curve {keys [
          {interpMode=RCIM_Cubic, leaveTangent=-0.0051, tangentMode=RCTM_Break, time=0, value=1}
          {arriveTangent=-0.247, interpMode=RCIM_Cubic, tangentMode=RCTM_Break, time=1, value=0.9}
         ]}}})
    )
    ApplyForce=false
    ApplyPosition=true
    ApplyPositionFalloff=false
   )
   &SolveForcesAndVelocity(Clamp Velocity=false, Limit Acceleration=false)
   &ScaleColor(
    Scale RGB=&VectorFromFloat(Value=&VC_FloatBySpawnGroup(
      SpawnGroup or ExecIndex:NiagaraInt32=0
      CustomInt:NiagaraInt32=0
      MaximumSpawnGroups:NiagaraInt32=4
      Float SG 0:NiagaraFloat=3
      Float SG 1:NiagaraFloat=3
      Float SG 2:NiagaraFloat=3
      Float SG 3:NiagaraFloat=3
      Float SG 4:NiagaraFloat=0
      Float SG 5:NiagaraFloat=0
     ))
    ScaleA=true
    ScaleRGB=true
   )
  ]
  rendererClasses [&NiagaraSpriteRendererProperties]
  renderers [
   {rendererClass=&NiagaraSpriteRendererProperties}
   {rendererClass=&NiagaraSpriteRendererProperties, rendererIndex=1}
   {rendererClass=&NiagaraSpriteRendererProperties, rendererIndex=2}
   {rendererClass=&NiagaraSpriteRendererProperties, rendererIndex=3}
  ]
 }
 Ribbon {
  emitterUpdateScript [
   &EmitterState(Inactive Response=::NewEnumerator1, Life Cycle Mode=::NewEnumerator1, Loop Duration=1, UseLoopDelay=false)
   &VC_RateQualityHandler(Epic Spawn Rate=30, High Spawn Rate=30, Low Spawn Rate=30, Use Custom Rate?=true)
   &VC_SpawnRibbon_v1(Number of Ribbons=2, SpawnRate=$Emitter.Spawn Rate)
  ]
  particleSpawnScript [
   &VC_PaintedVariantsOverride(
    Black=<rgba 0.0885 0.0885 0.0885 1.8>
    BurntSienna=<rgba 0.2344 0.1252 0.1086 2.2>
    Cobalt=<rgba 0 0.1345 0.8073 2.2>
    Crimson=<rgba 0.625 0 0.0169 2>
    DefaultAlpha=0.6
    DefaultColor=<rgba 0 0.5395 0.6514 1>
    ForestGreen=<rgba 0 0.6771 0.0342 2>
    Grey=<rgba 0.4323 0.4323 0.4323 2>
    Lime=<rgba 0.5078 0.6771 0 2>
    Orange=<rgba 0.672 0.3657 0.2016 2>
    Pink=<rgba 0.72 0.144 0.48 1.45>
    Purple=<rgba 0.5313 0 0.7969 2>
    R1 Black:LinearColor=<rgba 0.1927 0.1927 0.1927 50>
    R1 BurntSienna:LinearColor=<rgba 0.3229 0.1956 0.1615 40>
    R1 Cobalt:LinearColor=<rgba 0.0353 0.0659 1.6927 10>
    R1 Crimson:LinearColor=<rgba 2 0 0.0542 11>
    R1 DefaultColor=<rgba 1 1 0 20>
    R1 ForestGreen:LinearColor=<rgba 0 1.5 0 20>
    R1 Grey:LinearColor=<rgba 0.5469 0.5469 0.5469 35>
    R1 Lime:LinearColor=<rgba 0.6667 1 0 45>
    R1 Orange:LinearColor=<rgba 1 0.75 0 40>
    R1 Pink:LinearColor=<rgba 1 0.2865 0.7405 20>
    R1 Purple:LinearColor=<rgba 0.5885 0.2943 0.4815 40>
    R1 Saffron:LinearColor=<rgba 2 1.5333 0 30>
    R1 SkyBlue:LinearColor=<rgba 0.4 1.4646 2 15>
    R1 TitaniumWhite:LinearColor=<rgba 2 2 2 15>
    R2 Black:LinearColor=<rgba 0.155 0.155 0.155 1>
    R2 BurntSienna:LinearColor=<rgba 0.3281 0.192 0.1555 1>
    R2 Cobalt:LinearColor=<rgba 0 0.0555 3 0.2>
    R2 Crimson:LinearColor=<rgba 1 0.15 0.173 0.5>
    R2 DefaultColor:LinearColor=<rgba 1 1 1 0.3>
    R2 ForestGreen:LinearColor=<rgba 0.15 1 0.1929 0.35>
    R2 Grey:LinearColor=<rgba 0.5469 0.5469 0.5469 1>
    R2 Lime:LinearColor=<rgba 0.6667 1 0 0.8>
    R2 Orange:LinearColor=<rgba 1 0.5833 0 1>
    R2 Pink:LinearColor=<rgba 0.7917 0.2886 0.6087 0.5>
    R2 Purple:LinearColor=<rgba 0.5333 0 0.8 2>
    R2 Saffron:LinearColor=<rgba 1.5 1.15 0 1>
    R2 SkyBlue:LinearColor=<rgba 0.12 0.8386 1.2 1>
    Saffron=<rgba 0.9313 0.7468 0.1407 2>
    SkyBlue=<rgba 0 0.5988 0.9 1.8>
    TitaniumWhite=<rgba 0.7865 0.7865 0.7865 2>
    Use R1 Painted Overrides=true
    Use R2 Painted Overrides=true
   )
   &VC_InitializeRibbon_v1(
    Color=$StackContext.Color
    Custom Offset=true
    Enable Skeletal Position=true
    Lifetime=2.5
    Local Vector Offset=&AddVectorToPosition(Position=<8 0 -10>, Vector=$User.offset)
    Offset Amount=1
    Ribbon Width=30
   )
   &AddVelocity(Velocity=&MakeCustomVectorFromBool(Bool=$User.bIsFrontendPreview, True Vector=<-800 0 1>, False Vector=<-1500 0 0>)) {module=AddVelocity001}
  ]
  particleUpdateScript [
   &FN_DisableTireFX(Mask Index Attribute:NiagaraInt32=$Particles.SpawnGroup, Mask Index FL=2, Mask Index FR=3, Mask Index RL=0, Mask Index RR=1)
   &ParticleState()
   &InheritSourceMovement(
    Applied Position Scale=&VectorFromFloat(Value=&FloatFromCurve(FloatCurve=DataInterface() {propertyValues {Curve {keys [<at 0 1>,<at 1 0>]}}}))
    ApplyForce=false
    ApplyPosition=true
    ApplyPositionFalloff=false
   )
   &SolveForcesAndVelocity(Clamp Velocity=false, Limit Acceleration=false)
   &VC_UpdateRibbon_v1(Enable Skeletal Position=true)
   &ScaleRibbonWidth(Ribbon Width Scale=&FloatFromCurve(FloatCurve=DataInterface() {propertyValues {Curve {keys [<at 0 1.6>,<at 1 0.5>]}}})) {enabled=false}
   &DynamicMaterialParameters(
    Index 0 Param 1=&MakeCustomFloatFromBool(Bool=$User.bIsPainted, True Float=$User.ColorWhite, False Float=0)
    Index 0 Param 2=0.6
    Module.Index0Param0Write=true
    Module.Index0Param1Write=true
    Module.Index0Param2Write=true
    Module.Index0Param3Write=true
   ) {enabled=false}
   &ScaleColor(
    Scale Alpha=&FloatFromCurve(FloatCurve=DataInterface() {propertyValues {Curve {keys [<at 0 1>,<at 1 0>]}}})
    Scale RGB=&VectorFromFloat(Value=$User.PaintedScalar)
    ScaleA=true
    ScaleRGB=true
   ) {enabled=false}
   /SetVariables(Particles.ScaleOutLinecolor:LinearColor=&MultiplyLinearColor(
     Linear Color=$Particles.R1 Color
     Scale Factor=&MakeVector4FromVectorAndFloat(
      "Vector (RGB)"=<1.5 1.5 1.5>
      "Float (Alpha)"=&FloatFromCurve(FloatCurve=DataInterface() {propertyValues {Curve {
          defaultValue=1
          keys [
           {arriveTangent=-2.1875, interpMode=RCIM_Cubic, time=0, value=0}
           {interpMode=RCIM_Cubic, tangentMode=RCTM_User, time=0.2, value=1}
           {interpMode=RCIM_Cubic, tangentMode=RCTM_User, time=0.5, value=-0.05}
          ]
         }}})
     )
    )) {isSetParameters=true}
   /SetVariables(Particles.ScaleColorNoise:LinearColor=&MultiplyLinearColor(
     Linear Color=$Particles.R2 Color
     Scale Factor=&MakeVector4FromVectorAndFloat(
      "Vector (RGB)"=<1 1 1>
      "Float (Alpha)"=&FloatFromCurve(FloatCurve=DataInterface() {propertyValues {Curve {
          defaultValue=1
          keys [
           {arriveTangent=5, interpMode=RCIM_Cubic, time=0, value=0}
           {interpMode=RCIM_Cubic, tangentMode=RCTM_User, time=0.2, value=1}
           {interpMode=RCIM_Cubic, tangentMode=RCTM_User, time=0.5, value=0}
          ]
         }}})
     )
    )) {isSetParameters=true}
   /SetVariables(Particles.SizeOutlines:NiagaraFloat=&Multiply_Float(A=40, B=&FloatFromCurve(FloatCurve=DataInterface() {propertyValues {Curve {keys [<at 0 1>,<at 1 0>]}}}))) {isSetParameters=true}
   /SetVariables(Particles.float:NiagaraFloat=&Multiply_Float(A=35, B=&FloatFromCurve(FloatCurve=DataInterface() {propertyValues {Curve {keys [<at 0 1>,<at 1 0>]}}}))) {isSetParameters=true}
  ]
  rendererClasses [&NiagaraRibbonRendererProperties]
  renderers [
   {rendererClass=&NiagaraRibbonRendererProperties}
   {rendererClass=&NiagaraRibbonRendererProperties, rendererIndex=1}
   {rendererClass=&NiagaraRibbonRendererProperties, rendererIndex=2}
   {rendererClass=&NiagaraRibbonRendererProperties, rendererIndex=3}
  ]
 }
 Wheel {
  emitterSpawnScript [/SetVariables(
    Emitter.NeedReflection:NiagaraFloat=&SelectFloatFromArray(
     Direct Array Index=$System.ColorVariantID
     Float Selection Array=DataInterface() {propertyValues {FloatData [0,1,-0.15,-0.25,-0.25,-0.3,-0.5,-0.5,-0.5,-0.5,-0.5,-0.5,-0.5,-0.5]}}
    )
    Emitter.desaturation:NiagaraFloat=&SelectFloatFromArray(Direct Array Index=$System.ColorVariantID, Float Selection Array=DataInterface() {propertyValues {FloatData [0,1,1,1,1,1,1,1,1,1,1,1,1,1]}})
    Emitter.gain:NiagaraFloat=&SelectFloatFromArray(
     Direct Array Index=$System.ColorVariantID
     Float Selection Array=DataInterface() {propertyValues {FloatData [1.8,1,1.8,1.2,1.2,1,1,1,1,1,1,1,1,1]}}
    )
   ) {isSetParameters=true}]
  emitterUpdateScript [
   &EmitterState(Inactive Response=::NewEnumerator1, Life Cycle Mode=::NewEnumerator1, Loop Duration=1, UseLoopDelay=false)
   &SpawnRate(
    SpawnRate=&MakeCustomFloatFromBool(Bool=$User.bIsFrontendPreview, True Float=4, False Float=4)
    Use Spawn Probability=false
   )
   &SpawnBurst_Instantaneous(Spawn Count=2, Use Spawn Probability=false, UseLoopCountLimit=false) {enabled=false}
   &VC_WheelWrapScale()
  ]
  particleSpawnScript [
   &VC_PaintedVariantsOverride(
    Black=<rgba 0.016 0.016 0.016 0.2>
    BurntSienna=<rgba 0.0642 0.0198 0.0079 1>
    Crimson=<rgba 0.4635 0 0.0126 1>
    Orange=<rgba 0.6724 0.2346 0 1>
   )
   &InitializeParticle(
    Color=$StackContext.Color
    Mesh Scale=$Emitter.WheelFXScale
    Mesh Scale Mode=::NewEnumerator4
    Sprite Rotation Mode=::NewEnumerator2
    Sprite Size Mode=::NewEnumerator3
    UsePositionOffset=false
    Write Material Random=true
   )
   &PartitionParticles()
   &SkeletalMeshLocation(
    Bone / Socket Index=&Add_Integer(A=$Particles.PartitionParticles.Partition, B=0)
    Bone / Socket Sampling Mode=::NewEnumerator5
    Interpolate Mesh Sampling=false
    MaskSpawnGroup=false
    Mesh Sampling Type=::NewEnumerator6
    Skeletal Mesh=$System.SkeletalMesh
   )
  ]
  particleUpdateScript [
   &ParticleState()
   &SolveForcesAndVelocity(Clamp Velocity=false, Limit Acceleration=false)
   &UpdateMeshOrientation(Rotation Coordinate Space=::NewEnumerator2, Rotation Rate=1.5, Rotation Vector=<0 1 0>)
  ]
  rendererClasses [&NiagaraMeshRendererProperties]
  renderers [{rendererClass=&NiagaraMeshRendererProperties, rendererIndex=0}]
 }
}
summary {
 systemName=DriftTrail_StarBeam_MP
 emitters [
  {emitterName=Ribbon, rendererClasses [&NiagaraRibbonRendererProperties]}
  {emitterName=Wheel, rendererClasses [&NiagaraMeshRendererProperties]}
  {emitterName=AllStars, rendererClasses [&NiagaraSpriteRendererProperties]}
 ]
 userVariables [
  User.bIsFrontendPreview {type=NiagaraBool, defaultValue=true}
  User.bIsPainted {type=NiagaraBool}
  User.ColorVariantID {type=NiagaraInt32}
  User.ColorWhite {type=NiagaraFloat, defaultValue=0, description=""}
  User.offset {type=Vector3f, description="offset from wheels(set through cosmetic template)", defaultValue {struct=Vector3f, value=<0 0 -20>}}
  User.PaintedColor {type=LinearColor, description="", defaultValue {struct=LinearColor, value=<rgba 1 1 1 1>}}
  User.PaintedScalar {type=NiagaraFloat, defaultValue=1, description=""}
  User.Shape01 {type=LinearColor, description="", defaultValue {struct=LinearColor, value=<rgba 0 0.7784 1 1>}}
  User.Shape02 {type=LinearColor, description="", defaultValue {struct=LinearColor, value=<rgba 1 0.9 0 1>}}
  User.VehicleFlags {type=NiagaraInt32, defaultValue=806289408}
  User.Wheel_Radius_Array {
   type=NiagaraDataInterfaceArrayFloat
   defaultValue {
    struct=NiagaraExt_VariableValue_DataInterface
    value {
     dataInterface=&self:NiagaraDataInterfaceArrayFloat_0
     dataInterfaceClass=&NiagaraDataInterfaceArrayFloat
    }
   }
  }
  User.Wheel_Width_Array {
   type=NiagaraDataInterfaceArrayFloat
   defaultValue {
    struct=NiagaraExt_VariableValue_DataInterface
    value {
     dataInterface=&self:NiagaraDataInterfaceArrayFloat_1
     dataInterfaceClass=&NiagaraDataInterfaceArrayFloat
    }
   }
  }
  User.WheelSize {type=NiagaraFloat, defaultValue=0, description=""}
 ]
}
